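# Prototype master material for hard-surface glow assets in Athena, using the Default Lit shading model and flagged for skeletal mesh and mesh deformer use.
# It exposes a Normal texture parameter, defaulting to a flat normal map, and a light grey Color vector parameter.
# The graph shows no links from these parameters to material outputs and no dedicated glow logic.
# Automatic usage flagging in the editor is off.
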Material /Game/Athena/Prototype/HardSurface/Materials/Hardsurface_Glow_Proto_Master
graph {
 properties {
  bAutomaticallySetUsageInEditor=false
  bUsedWithMeshDeformer=true
  bUsedWithSkeletalMesh=true
  pixelDepthOffsetMode=PDOM_Legacy
  usedShadingModels=Default Lit
 }
 nodes [
  TextureSampleParameter2D() {parameterName=Normal, samplerType=SAMPLERTYPE_Normal, texture=/Game/Athena/Prototype/HardSurface/Textures/HS_FlatNormal}
  VectorParameter() {parameterName=Color, defaultValue=<rgba 0.8177 0.8177 0.8177 1>}
 ]
}
